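# Sound cue for the front-end UI level button press screen-in sound, stored among the toast sounds.
# A single wave player plays one sound wave, routed to the Fort_UI sound class.
# A concurrency override allows only one instance at a time, stopping the oldest when it retriggers.
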SoundCue /Game/Sounds/Fort_UI_Sounds/FrontEnd/New_FrontEnd/Toasts/SC_UI_LVL_Button_Pressed_Screen_In
root=WavePlayer_3
nodes [{id=WavePlayer_3, properties {soundWaveAssetPtr=&/UI_LVL_Button_Press_Screen_In}}]
properties {
 bOverrideConcurrency=true
 maxDistance=2097152
 soundClassObject=&Fort_UI
 volumeMultiplier=1
 concurrencyOverrides {maxCount=1, resolutionRule=StopOldest}
}
