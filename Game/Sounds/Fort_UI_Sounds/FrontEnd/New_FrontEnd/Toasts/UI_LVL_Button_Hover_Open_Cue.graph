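# UI sound cue for hovering open a level button in the front-end toast set, judging by its name and path.
# A single wave player feeds the root directly, playing the matching hover-open sound at reduced volume in the Fort_UI sound class.
# An expandable level UI concurrency set is assigned, alongside override values for a single voice that stops the oldest.
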SoundCue /Game/Sounds/Fort_UI_Sounds/FrontEnd/New_FrontEnd/Toasts/UI_LVL_Button_Hover_Open_Cue
root=WavePlayer_3
nodes [{id=WavePlayer_3, properties {soundWaveAssetPtr=&/UI_LVL_Button_Hover_Open}}]
properties {
 maxDistance=2097152
 soundClassObject=&Fort_UI
 volumeMultiplier=0.8
 concurrencyOverrides {maxCount=1, resolutionRule=StopOldest}
 concurrencySet [&/Conc_Expandable_Level_UI]
}
